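# Sound cue for Smackie creature swimming foley, played through a creature-steps sound class with the Prey_Gallop attenuation.
# A random node picks among four equally weighted third-person player swim-sprint water waves.
# An enveloper applies a short volume fade-in to soften the attack, then a modulator randomizes pitch to nearly double, repurposing the player splashes as higher, quicker creature strokes.
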SoundCue /Irwin/Sounds/Smackie/Foley/Smackie_Swim_Cue
root=Modulator_0
nodes [
 {
  id=Enveloper_0
  children [Random_3]
  properties {pitchCurve {editorCurveData {keys [{time=0}]}}, volumeCurve {editorCurveData {keys [<at 0 0.0109>,<at 0.3 1>]}}}
 }
 {id=Modulator_0, children [Enveloper_0], properties {pitchMax=2, pitchMin=1.8}}
 {id=Random_3, children [WavePlayer_18, WavePlayer_19, WavePlayer_20, WavePlayer_21], properties {weights [1,1,1,1]}}
 {id=WavePlayer_18, properties {soundWaveAssetPtr=/Game/Athena/Sounds/Water/Interaction/Water_Player_Swim_Sprint_3P_08}}
 {id=WavePlayer_19, properties {soundWaveAssetPtr=/Game/Athena/Sounds/Water/Interaction/Water_Player_Swim_Sprint_3P_02}}
 {id=WavePlayer_20, properties {soundWaveAssetPtr=/Game/Athena/Sounds/Water/Interaction/Water_Player_Swim_Sprint_3P_03}}
 {id=WavePlayer_21, properties {soundWaveAssetPtr=/Game/Athena/Sounds/Water/Interaction/Water_Player_Swim_Sprint_3P_09}}
]
properties {
 attenuationSettings=/Irwin/Sounds/Settings/Prey_Gallop_Att
 maxDistance=3200
 soundClassObject=/Game/Sounds/FortSoundClassesAndModes/Fort_Creatures_Steps_Normal
 volumeMultiplier=0.6
}
